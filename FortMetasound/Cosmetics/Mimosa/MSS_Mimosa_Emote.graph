MetaSoundSource /FortMetasound/Cosmetics/Mimosa/MSS_Mimosa_Emote
inputs [
 B {type=Enum:FortAudioAffiliation}
 Character.Affiliation {type=Enum:FortAudioAffiliation}
 Character.Health {type=Float}
 Character.UniqueID {type=Int32}
 DelayLoopByStart {type=Bool, access=Constructor}
 FadeInTime {type=Time, access=Constructor}
 FriendlyVolume {type=Float}
 Intro.Music.Sounds {type=WaveAsset:Array, access=Constructor}
 IntroVolumeMultiplier {type=Float}
 Loop.Music.Sounds {type=WaveAsset:Array, access=Constructor}
 Looping {type=Bool, access=Constructor}
 LoopVolumeMultiplier {type=Float}
 PitchTemplateMultiplier {type=Float}
 UE.Source.OnPlay {type=Trigger}
 UnfriendlyVolume {type=Float}
]
interfaces [Fort.Character:1.0, UE.OutputFormat.Stereo:1.0, UE.Source:1.1]
outputs [
 "UE.OutputFormat.Stereo.Audio:0" {type=Audio}
 "UE.OutputFormat.Stereo.Audio:1" {type=Audio}
 UE.Source.OneShot.OnFinished {type=Trigger}
 VoEnvelope {type=Float}
]
pages [Default {
  input_defaults {
   B=null
   Character.Affiliation=null
   Character.Health=null
   Character.UniqueID=null
   DelayLoopByStart="<MetaSoundBuilderResult.SUCCEEDED: 0>"
   FadeInTime=0
   FriendlyVolume=null
   IntroVolumeMultiplier=1
   LoopVolumeMultiplier=1
   Looping=null
   PitchTemplateMultiplier=1
   UE.Source.OnPlay=null
   UnfriendlyVolume=null
   Intro.Music.Sounds []
   Loop.Music.Sounds []
  }
  nodes [
   InitVariable.Audio()
   InitVariable.Enum:FortAudioAffiliation()
   InitVariable.Float()
   InitVariable.Trigger()
   InitVariable.Trigger()
   InitVariable.Trigger()
   InitVariable.Trigger()
   InitVariable.Trigger()
   ADSREnvelopeADSREnvelope_1 = ADSR Envelope.ADSR Envelope:Float(Trigger Attack=^UE.Source.OnPlay, Trigger Release=VariableAccessor.Trigger(Variable=InitVariable.Trigger().Variable).Value) {also {defaults {Attack Time=0, Decay Time=0, Release Time=0.1, Sustain Level=1}}}
   n1 = InitVariable.Audio()
   ADSREnvelopeADSREnvelope_0 = ADSR Envelope.ADSR Envelope:Float(Attack Time=^FadeInTime, Trigger Attack=^UE.Source.OnPlay) {also {defaults {Sustain Level=1}}}
   n2 = VariableAccessor.Float(Variable=VariableMutator.Float(Value=@ADSREnvelopeADSREnvelope_0.Out Envelope, Variable=InitVariable.Float().Variable).Variable)
   n3 = InitVariable.Audio()
   ArrayRandomGet_0 = Array.Random Get:WaveAsset:Array(In Array=^Intro.Music.Sounds, Next=^UE.Source.OnPlay)
   FrequencyMultipliertoSemitone_1 = UE.Frequency Multiplier to Semitone:Float(Frequency Multiplier=^PitchTemplateMultiplier)
   WavePlayer_0 = UE.Wave Player:Stereo(Play=@ArrayRandomGet_0.On Next, Wave Asset=@ArrayRandomGet_0.Value, Pitch Shift=@FrequencyMultipliertoSemitone_1.Semitones) {also {version [1,1]}}
   n4 = InitVariable.Audio()
   n5 = InitVariable.Trigger()
   n6 = VariableMutator.Trigger(Variable=InitVariable.Trigger().Variable, Value=@WavePlayer_0.On Finished)
   TriggerAnyTriggerAny2_0 = "TriggerAny.Trigger Any (2)"(
    In 0=TriggerCompare.Clamp:Bool(A=^DelayLoopByStart, Compare=^UE.Source.OnPlay).TRUE
    In 1=TriggerCompare.Clamp:Bool(A=^DelayLoopByStart, Compare=VariableAccessor.Trigger(Variable=@n6.Variable).Value) {also {defaults {B="<MetaSoundBuilderResult.SUCCEEDED: 0>"}}}.TRUE
   )
   ArrayRandomGet_1 = Array.Random Get:WaveAsset:Array(
    In Array=UE.Reroute(Value=^Loop.Music.Sounds).Value
    Next=VariableAccessor.Trigger(Variable=VariableMutator.Trigger(Variable=@n5.Variable, Value=@TriggerAnyTriggerAny2_0.Out).Variable).Value
   )
   FrequencyMultipliertoSemitone_0 = UE.Frequency Multiplier to Semitone:Float(Frequency Multiplier=^PitchTemplateMultiplier)
   WavePlayer_1 = UE.Wave Player:Stereo(Loop=^Looping, Play=@ArrayRandomGet_1.On Next, Wave Asset=@ArrayRandomGet_1.Value, Pitch Shift=@FrequencyMultipliertoSemitone_0.Semitones) {also {version [1,1]}}
   n7 = InitVariable.Audio()
   n8 = InitVariable.Audio()
   AudioMixerAudioMixerStereo2_0 = "AudioMixer.Audio Mixer (Stereo, 2)"(
    Gain 1=UE.Multiply:Float(AdditionalOperands=^IntroVolumeMultiplier, PrimaryOperand=@n2.Value).Out
    Gain 0=UE.Multiply:Float(AdditionalOperands=^LoopVolumeMultiplier, PrimaryOperand=VariableAccessor.Float(Variable=@n2.Variable).Value).Out
    In 1 L=VariableAccessor.Audio(Variable=VariableMutator.Audio(Variable=@n3.Variable, Value=@WavePlayer_0.Out Left).Variable).Value
    In 0 L=VariableAccessor.Audio(Variable=VariableMutator.Audio(Variable=@n4.Variable, Value=@WavePlayer_1.Out Left).Variable).Value
    In 0 R=VariableAccessor.Audio(Variable=VariableMutator.Audio(Variable=@n7.Variable, Value=@WavePlayer_1.Out Right).Variable).Value
    In 1 R=VariableAccessor.Audio(Variable=VariableMutator.Audio(Variable=@n8.Variable, Value=@WavePlayer_0.Out Right).Variable).Value
   )
   n9 = UE.Multiply:Audio by Float(PrimaryOperand=UE.Add:Audio(PrimaryOperand=@AudioMixerAudioMixerStereo2_0.Out L, AdditionalOperands=@AudioMixerAudioMixerStereo2_0.Out R).Out) {also {defaults {AdditionalOperands=0.5}}}
   EnvelopeFollower_0 = UE.Envelope Follower(In=VariableAccessor.Audio(Variable=VariableMutator.Audio(Variable=@n1.Variable, Value=@n9.Out).Variable).Value) {also {defaults {Peak Mode=2}}}
   Multiply_3 = UE.Multiply:Float(
    PrimaryOperand=@ADSREnvelopeADSREnvelope_1.Out Envelope
    AdditionalOperands=MapRange.MapRange:Float(In=@EnvelopeFollower_0.Envelope) {also {defaults {In Range B=0.2}}}.Out Value
   )
   VariableMutator.Trigger(Variable=InitVariable.Trigger().Variable, Value=@WavePlayer_1.On Finished)
  ]
  links [
   {dst="@UE.OutputFormat.Stereo.Audio:0", src=AudioMixerAudioMixerStereo2_0.Out L}
   {dst="@UE.OutputFormat.Stereo.Audio:1", src=AudioMixerAudioMixerStereo2_0.Out R}
   {dst="@VoEnvelope", src=Multiply_3.Out}
  ]
 }]
properties {
 assetImportData=&self:AssetImportData
 attenuationSettings=&Fort_Emotes_Athena_Att
 audioPropertiesSheet=/CosmeticCompanions/Assets/Audio_Shared/PropertySheets/APS_Cosmetic_Mimosa_Emote
 bLooping=true
 duration=10000
 maxDistance=1600
 numChannels=2
 outputFormat=Stereo
 sampleRate=48000
 soundClassObject=&Fort_Creatures_Pets
 virtualizationMode=PlayWhenSilent
 concurrencySet [/Game/Sounds/Concurrency_Rules/Athena_EmoteMusic_Conc]
 soundSubmixSends [{}]
}
