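# MetaSound Source preset that plays emote foley for the WittyStork biped cosmetic companion, built on the shared MSS_Mimosa_Emote parent.
# It feeds Emote_WittyStork_SFX into the intro slot and leaves the loop slot empty, while the parent supplies the playback logic.
# Affiliation-related inputs set a lower volume for unfriendly than friendly.
# Stereo output uses emote attenuation, the pets sound class, emote-music concurrency and play-when-silent virtualization.
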MetaSoundSource /CosmeticCompanions/Assets/Biped/WittyStork/Audio/Emote/MSS_Mimosa_WittyStork_EmoteFoley
preset_parent=/FortMetasound/Cosmetics/Mimosa/MSS_Mimosa_Emote
inputs [
 B {type=Enum:FortAudioAffiliation}
 Character.Affiliation {type=Enum:FortAudioAffiliation}
 Character.Health {type=Float}
 Character.UniqueID {type=Int32}
 DelayLoopByStart {type=Bool, access=Constructor}
 FadeInTime {type=Time, access=Constructor}
 FriendlyVolume {type=Float}
 Intro.Music.Sounds {type=WaveAsset:Array, access=Constructor}
 IntroVolumeMultiplier {type=Float}
 Loop.Music.Sounds {type=WaveAsset:Array, access=Constructor}
 Looping {type=Bool, access=Constructor}
 LoopVolumeMultiplier {type=Float}
 PitchTemplateMultiplier {type=Float}
 UE.Source.OnPlay {type=Trigger}
 UnfriendlyVolume {type=Float}
]
interfaces [Fort.Character:1.0, UE.OutputFormat.Stereo:1.0, UE.Source:1.1]
outputs [
 "UE.OutputFormat.Stereo.Audio:0" {type=Audio}
 "UE.OutputFormat.Stereo.Audio:1" {type=Audio}
 UE.Source.OneShot.OnFinished {type=Trigger}
 VoEnvelope {type=Float}
]
pages [Default {input_defaults {
   B=1
   Character.Affiliation=0
   Character.Health=0
   Character.UniqueID=0
   DelayLoopByStart="<MetaSoundBuilderResult.SUCCEEDED: 0>"
   FadeInTime=0
   FriendlyVolume=1
   IntroVolumeMultiplier=1
   LoopVolumeMultiplier=1
   Looping=null
   PitchTemplateMultiplier=1
   UE.Source.OnPlay=null
   UnfriendlyVolume=0.6
   Intro.Music.Sounds [&/Emote_WittyStork_SFX]
   Loop.Music.Sounds []
  }}]
properties {
 assetImportData=&self:AssetImportData
 attenuationSettings=&Fort_Emotes_Athena_Att
 audioPropertiesSheet=/CosmeticCompanions/Assets/Audio_Shared/PropertySheets/APS_Cosmetic_Mimosa_Emote
 bLooping=true
 duration=10000
 maxDistance=1600
 numChannels=2
 outputFormat=Stereo
 sampleRate=48000
 soundClassObject=&Fort_Creatures_Pets
 virtualizationMode=PlayWhenSilent
 concurrencySet [/Game/Sounds/Concurrency_Rules/Athena_EmoteMusic_Conc]
 soundSubmixSends [{}]
}
